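# Niagara system for the GraffitiTon player emote, drawing pink and yellow spray-paint effects attached to skeletal mesh bones or sockets.
# Ribbon paint and smoke trails, smoke sprites, burst puffs and GPU drops use spawn groups to choose socket and color, gated by User float parameters.
# Spray meshes follow their sockets with User-driven alpha, while small fading paint decals spawn continuously around one socket.
# Most emitters tint color by a system time-of-day value.
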NiagaraSystem /BRCosmetics/Animation/Game/MainPlayer/Emotes/GraffitiTon/FX/NS_GraffitiTon_Trail
dependencies {
 usedDataInterfaces [
  &NiagaraDataInterfaceCurve
  &NiagaraDataInterfaceColorCurve
  &NiagaraDataInterfaceVector2DCurve
  &NiagaraDataInterfaceSpriteRendererInfo
  &NiagaraDataInterfaceArrayColor
  &NiagaraDataInterfaceSkeletalMesh
 ]
 usedDynamicInputs [&FNTimeOfDayColor]
 usedModules [
  &self:SystemSpawnScript.SystemScriptSource.SystemScriptGraph.NiagaraNodeAssignment_1.SetVariables_#1
  &self:SystemSpawnScript.SystemScriptSource.SystemScriptGraph.NiagaraNodeAssignment_6.SetVariables_#2
  &SystemState
  &self:SystemSpawnScript.SystemScriptSource.SystemScriptGraph.NiagaraNodeAssignment_0.SetVariables_#3
 ]
}
emitters {
 BurstSmoke {
  emitterUpdateScript [
   &EmitterState()
   &SpawnRate(SpawnRate=&Multiply_Float(A=24, B=$User.Burst), Use Spawn Probability=false) {module=SpawnRate001}
   &SpawnRate(SpawnGroup=1, SpawnRate=&Multiply_Float(A=24, B=$User.Burst), Use Spawn Probability=false)
   &SpawnRate(SpawnGroup=2, SpawnRate=&Multiply_Float(A=24, B=$User.Burst2), Use Spawn Probability=false) {module=SpawnRate002}
   &SpawnRate(SpawnGroup=3, SpawnRate=&Multiply_Float(A=24, B=$User.Burst2), Use Spawn Probability=false) {module=SpawnRate003}
  ]
  particleSpawnScript [
   &InitializeParticle(
    Color=&MaskLinearColorBySpawnGroup(
     Masked Linear Color=<rgba 1 0.0782 0.8714 1>
     Passthrough Linear Color=&MaskLinearColorBySpawnGroup(
      Masked Linear Color=<rgba 1 0.6445 0 1>
      Passthrough Linear Color=&MaskLinearColorBySpawnGroup(Masked Linear Color=<rgba 1 0.0778 0.8737 1>, Passthrough Linear Color=<rgba 1 0.6469 0 1>, Spawn Group Mask=2)
      Spawn Group Mask=1
     )
    )
    Lifetime=&RandomRangeFloat(Minimum=0.7, Maximum=1)
    Sprite Rotation Mode=::NewEnumerator2
    Sprite Size Mode=::NewEnumerator1
    Sprite UV Mode=::NewEnumerator3
    Uniform Sprite Size Max=50
    Uniform Sprite Size Min=35
    UsePositionOffset=false
    Write Material Random=true
   )
   &SkeletalMeshLocation(
    Bone / Socket Index=&SpawnGroup()
    Bone Sampling Mode=::NewEnumerator4
    MaskSpawnGroup=false
    Skeletal Mesh=$System.Skeletal Mesh
   )
   &AddVelocity(
    Rotation Coordinate Space=::Simulation
    Velocity=&TransformVector(
     Vector=&MaskVectorBySpawnGroup(
      Masked Vector=<-1 0.5 0.3>
      Passthrough Vector=&MaskVectorBySpawnGroup(
       Masked Vector=<1 0.3 0.3>
       Passthrough Vector=&MaskVectorBySpawnGroup(Masked Vector=<1 0.7 0.3>, Passthrough Vector=<-1 0.8 0.3>, Spawn Group Mask=2)
       Spawn Group Mask=1
      )
     )
     Destination Space=::Local
    )
    Velocity Speed Scale=&Multiply_Float(A=&RandomRangeFloat(Minimum=125, Maximum=180), B=1)
   )
   &ApplyInitialForces()
  ]
  particleUpdateScript [
   &ParticleState()
   &ScaleSpriteSize(Uniform Curve Sprite Scale=DataInterface() {propertyValues {Curve {keys [
        {arriveTangent=0.4427, interpMode=RCIM_Cubic, leaveTangent=0.4427, tangentMode=RCTM_User, time=0, value=0.2}
        {arriveTangent=2.9153, interpMode=RCIM_Cubic, leaveTangent=2.9153, tangentMode=RCTM_User, time=0.4011, value=0.7756}
        <at 1 1.3821 RCIM_Cubic>
       ]}}})
   &ScaleColor(
    Scale Alpha=&FloatFromCurve(FloatCurve=DataInterface() {propertyValues {Curve {
        defaultValue=1
        keys [
         {arriveTangent=9.0725, interpMode=RCIM_Cubic, leaveTangent=9.0725, tangentMode=RCTM_User, time=0, value=0}
         {arriveTangent=-2.5249, interpMode=RCIM_Cubic, leaveTangent=-2.5249, tangentMode=RCTM_User, time=0.1408, value=0.9847}
         {arriveTangent=-0.0405, interpMode=RCIM_Cubic, leaveTangent=-0.0405, tangentMode=RCTM_User, time=1, value=0}
        ]
       }}})
    Scale RGB=&Multiply_Vector(A=<1 1 1>, B=&MakeVectorFromLinearColorRGB(Color=$System.ToD))
    ScaleA=true
    ScaleRGB=true
   )
   /SetVariables(Particles.SubImageIndex=&FloatFromCurve(FloatCurve=DataInterface() {propertyValues {Curve {keys [<at 0 30>,<at 1 64>]}}})) {isSetParameters=true}
   &DynamicMaterialParameters(
    Index 0 Param 1=3
    Index 0 Param 2=1
    Index 0 Param 3=0.5
    Module.Index0Param0Write=true
    Module.Index0Param1Write=true
    Module.Index0Param2Write=true
    Module.Index0Param3Write=true
   )
   &InheritSourceMovement(
    Applied Position Falloff Distance=1000
    Applied Position Scale=&Multiply_VectorByFloat(
     Vector=<0.6 0.6 0>
     Float=&FloatFromCurve(
      FloatCurve=DataInterface() {propertyValues {Curve {keys [
          <at 0 0 RCIM_Cubic>
          {interpMode=RCIM_Cubic, tangentMode=RCTM_User, time=0.1865, value=1}
          {arriveTangent=-2.1905, interpMode=RCIM_Cubic, leaveTangent=-2.1905, tangentMode=RCTM_User, time=1, value=0}
         ]}}}
      CurveIndex=$Emitter.Age
     )
    )
    ApplyForce=false
    ApplyPosition=true
    ApplyPositionFalloff=false
   )
   &Drag(
    Drag=&FloatFromCurve(
     FloatCurve=DataInterface() {propertyValues {Curve {keys [
         {interpMode=RCIM_Cubic, tangentMode=RCTM_User, time=0, value=0}
         {arriveTangent=22.0021, interpMode=RCIM_Cubic, leaveTangent=22.0021, tangentMode=RCTM_User, time=0.06, value=1}
        ]}}}
     Scale Curve=3
    )
    Use Linear Drag=true
    Use Rotational Drag=true
   ) {module=Drag002}
   &SolveForcesAndVelocity(Clamp Velocity=false, Limit Acceleration=false)
  ]
  rendererClasses [&NiagaraSpriteRendererProperties]
  renderers [{rendererClass=&NiagaraSpriteRendererProperties}]
 }
 Decal {
  emitterUpdateScript [
   &EmitterState(Life Cycle Mode=::NewEnumerator1, Loop Behavior=::NewEnumerator1, Loop Duration Mode=::NewEnumerator1, UseLoopDelay=false)
   &SpawnRate(SpawnRate=12, Use Spawn Probability=false)
  ]
  particleSpawnScript [
   &InitializeParticle(
    Color=&SelectLinearColorFromArray(
     Array Sampling Mode=::NewEnumerator0
     Direct Array Index=?
     Color Selection Array=DataInterface() {propertyValues {ColorData [<rgba 1 0.1161 0.7858 1>, <rgba 1 0.8148 0 1>]}}
     Random Seed=0
    )
    Lifetime Max=1.5
    Lifetime Min=0.7
    Lifetime Mode=::NewEnumerator1
    Mass Max=1.5
    Mass Mode=::NewEnumerator3
    Sprite Rotation Mode=::NewEnumerator2
    Sprite Size Mode=::NewEnumerator1
    Uniform Sprite Size Max=6
    Uniform Sprite Size Min=2
    UsePositionOffset=false
    Write Material Random=true
   )
   &SkeletalMeshLocation(
    Bone / Socket Index=6
    Bone Sampling Mode=::NewEnumerator4
    MaskSpawnGroup=false
    Sampled Position Offset=<0 -50 -100>
    Skeletal Mesh=$System.Skeletal Mesh
   )
   &ShapeLocation(Disc Coverage=0.85, MaskSpawnGroup=false, Override Seed=false, Shape Primitive=::NewEnumerator4, UseOwnerScale=true)
  ]
  particleUpdateScript [
   &ParticleState()
   /SetVariables(Particles.DecalSize:Vector3f=&Multiply_Vector(
     A=&VectorFromFloat(Value=&RandomRangeFloat(Minimum=5, Maximum=27.5))
     B=&VectorFromFloat(Value=&FloatFromCurve(FloatCurve=DataInterface() {propertyValues {Curve {keys [
           {arriveTangent=4.1785, interpMode=RCIM_Cubic, leaveTangent=4.1785, tangentMode=RCTM_User, time=0, value=0.5}
           {arriveTangent=1.1179, interpMode=RCIM_Cubic, leaveTangent=1.1179, tangentMode=RCTM_User, time=0.01, value=0.8}
           {arriveTangent=0.0224, interpMode=RCIM_Cubic, leaveTangent=0.0224, tangentMode=RCTM_User, time=0.05, value=1}
          ]}}}))
    )) {isSetParameters=true}
   /SetVariables(Particles.DecalFade:NiagaraFloat=&FloatFromCurve(FloatCurve=DataInterface() {propertyValues {Curve {keys [<at 0 1>,<at 1 0>]}}})) {isSetParameters=true}
   &ScaleColor(
    Scale Alpha=&FloatFromCurve(FloatCurve=DataInterface() {propertyValues {Curve {keys [<at 0 1>,<at 1 0>]}}})
    ScaleA=true
    ScaleRGB=true
   )
  ]
  rendererClasses [&NiagaraDecalRendererProperties]
  renderers [{rendererClass=&NiagaraDecalRendererProperties, rendererIndex=0}]
 }
 Drops {
  simTarget=GPUComputeSim
  emitterUpdateScript [
   &EmitterState()
   &SpawnRate(SpawnRate=&Multiply_Float(A=6, B=$User.Trail), Use Spawn Probability=true)
   &SpawnRate(SpawnGroup=1, SpawnRate=&Multiply_Float(A=6, B=$User.Trail2), Use Spawn Probability=true) {module=SpawnRate001}
   &SpawnRate(SpawnGroup=2, SpawnRate=&Multiply_Float(A=1, B=$User.Trail), Use Spawn Probability=true) {module=SpawnRate002}
   &SpawnRate(SpawnGroup=3, SpawnRate=&Multiply_Float(A=1, B=$User.Trail2), Use Spawn Probability=true) {module=SpawnRate003}
  ]
  particleSpawnScript [
   &InitializeParticle(
    Color=&MaskLinearColorBySpawnGroup(Masked Linear Color=<rgba 1 0.2483 0.2961 1>, Passthrough Linear Color=<rgba 1 0.9361 0 1>)
    Lifetime Max=0.6
    Lifetime Min=0.2
    Lifetime Mode=::NewEnumerator1
    Sprite Size=&MaskVector2dBySpawnGroup(
     Masked Vector2D=&RandomRangeVector2D(Minimum=<8 8>, Maximum=<12 12>)
     Passthrough Vector2D=&MaskVector2dBySpawnGroup(
      Masked Vector2D=&RandomRangeVector2D(Minimum=<8 8>, Maximum=<12 12>)
      Passthrough Vector2D=&MaskVector2dBySpawnGroup(Masked Vector2D=<3 36>, Passthrough Vector2D=<5 45>, Spawn Group Mask=2)
      Spawn Group Mask=1
     )
     Spawn Group Mask=0
    )
    Sprite Size Mode=::NewEnumerator4
    UsePositionOffset=false
    Write Material Random=true
   ) {module=InitializeParticle001}
   &SkeletalMeshLocation(
    Bone / Socket Index=&SpawnGroup()
    Bone Sampling Mode=::NewEnumerator4
    Inherited Velocity Speed Limit=25
    MaskSpawnGroup=false
    Position Sampling=::NewEnumerator1
    Sampled Position Offset=&MaskVectorBySpawnGroup(
     Masked Vector=<0 0 0>
     Passthrough Vector=&MaskVectorBySpawnGroup(
      Masked Vector=<0 0 0>
      Passthrough Vector=&MaskVectorBySpawnGroup(Masked Vector=<0 0 -15>, Passthrough Vector=<0 0 -15>, Spawn Group Mask=2)
      Spawn Group Mask=1
     )
    )
    Skeletal Mesh=$System.Skeletal Mesh
    Velocity Sampling=::NewEnumerator0
    VelocitySpeedLimit=true
   )
   &AddVelocity(
    Cone Angle=60
    Cone Axis=&ReturnMeshOrientationAxis(Mesh Orientation Axis=::XAxis, Mesh Orientation Quat=$Particles.MeshOrientation)
    Module.UseConeFalloff=false
    Module.UseVelDistribution=false
    Override Seed=false
    Velocity Mode=::NewEnumerator2
    Velocity Speed=&MaskFloatBySpawnGroup(
     Masked Float=&RandomRangeFloat(Minimum=-325, Maximum=-500)
     Passthrough Float=&MaskFloatBySpawnGroup(
      Masked Float=&RandomRangeFloat(Minimum=-325, Maximum=-500)
      Passthrough Float=&MaskFloatBySpawnGroup(Masked Float=-45, Passthrough Float=-80, Spawn Group Mask=2)
      Spawn Group Mask=1
     )
    )
   ) {module=AddVelocity002}
   &SubUVAnimation(
    End Frame Range Override=3
    "Sprite Renderer "=DataInterface() {propertyValues {SpriteRenderer=&self:Drops_6.NiagaraSpriteRendererProperties_2}}
    SubUV Animation Mode:ENiagara_SubUVLookupModeV2=::NewEnumerator2
    UseEndFrame=true
    UseStartFrame=false
   )
  ]
  particleUpdateScript [
   &ParticleState()
   &GravityForce(
    Coordinate Space=::Local
    Gravity=&MakeVector(
     X=0
     Y=0
     Z=&MaskFloatBySpawnGroup(
      Masked Float=-980
      Passthrough Float=&MaskFloatBySpawnGroup(
       Masked Float=-980
       Passthrough Float=&MaskFloatBySpawnGroup(Masked Float=-1500, Passthrough Float=-1500, Spawn Group Mask=2)
       Spawn Group Mask=1
      )
     )
    )
   )
   &Drag(Drag=&RandomRangeFloat(Minimum=3, Maximum=5), Use Linear Drag=true, Use Rotational Drag=false)
   &SolveForcesAndVelocity(Clamp Velocity=false, Limit Acceleration=false)
   &ScaleColor(
    Scale Alpha=&FloatFromCurve(FloatCurve=DataInterface() {propertyValues {Curve {keys [
         <at 0 0 RCIM_Cubic>
         {arriveTangent=1.6041, leaveTangent=1.6041, time=0.05, value=1.0113}
         {interpMode=RCIM_Cubic, leaveTangent=-1.0633, time=0.75, value=1}
         <at 0.99 0 RCIM_Cubic>
        ]}}})
    Scale RGB=&Multiply_Vector(A=<1 1 1>, B=&MakeVectorFromLinearColorRGB(Color=$System.ToD))
    ScaleA=true
    ScaleRGB=true
   )
   &ScaleSpriteSize(Scale Factor=&Vector2DFromFloat(Value=&MaskFloatBySpawnGroup(
      Masked Float=&FloatFromCurve(FloatCurve=DataInterface() {propertyValues {Curve {keys [
           {arriveTangent=-0.4919, interpMode=RCIM_Cubic, leaveTangent=-0.4919, tangentMode=RCTM_User, time=0.25, value=2}
           {arriveTangent=-3.3301, interpMode=RCIM_Cubic, leaveTangent=-3.3302, tangentMode=RCTM_User, time=1, value=0}
          ]}}})
      Passthrough Float=&MaskFloatBySpawnGroup(
       Masked Float=&FloatFromCurve(FloatCurve=DataInterface() {propertyValues {Curve {keys [
            {arriveTangent=-0.4919, interpMode=RCIM_Cubic, leaveTangent=-0.4919, tangentMode=RCTM_User, time=0.25, value=2}
            {arriveTangent=-3.3301, interpMode=RCIM_Cubic, leaveTangent=-3.3302, tangentMode=RCTM_User, time=1, value=0}
           ]}}})
       Passthrough Float=&FloatFromCurve(FloatCurve=DataInterface() {propertyValues {Curve {keys [
            {arriveTangent=14.443, interpMode=RCIM_Cubic, leaveTangent=14.443, tangentMode=RCTM_User, time=0, value=0}
            <at 0.2 2 RCIM_Cubic>
            {arriveTangent=-3.3301, interpMode=RCIM_Cubic, leaveTangent=-3.3302, tangentMode=RCTM_User, time=1, value=0}
           ]}}})
       Spawn Group Mask=1
      )
     )))
   &ScaleSpriteSizeBySpeed(
    Max Scale Factor=&MaskVector2dBySpawnGroup(
     Masked Vector2D=&RandomRangeVector2D(Minimum=<0.2 1.5>, Maximum=<0.5 3>)
     Passthrough Vector2D=&MaskVector2dBySpawnGroup(
      Masked Vector2D=&RandomRangeVector2D(Minimum=<0.2 1.5>, Maximum=<0.5 3>)
      Passthrough Vector2D=&MaskVector2dBySpawnGroup(Masked Vector2D=<1 1>, Passthrough Vector2D=<1 1>, Spawn Group Mask=2)
      Spawn Group Mask=1
     )
     Spawn Group Mask=0
    )
    Velocity Threshold=325
   )
   &DynamicMaterialParameters(
    Index 0 Param 1=&FloatFromCurve(FloatCurve=DataInterface() {propertyValues {Curve {keys [<at 0 1 RCIM_Cubic>,<at 1 0 RCIM_Cubic>]}}})
    Module.Index0Float0Write=true
    Module.Index0Float1Write=true
    Module.Index0Float2Write=true
    Module.Index0Float3Write=true
   )
  ]
  rendererClasses [&NiagaraSpriteRendererProperties]
  renderers [{rendererClass=&NiagaraSpriteRendererProperties}]
 }
 MainTrail {
  emitterUpdateScript [
   &EmitterState()
   &SpawnRate(SpawnRate=&Multiply_Float(A=$System.TrailSpawnRate, B=$User.Trail), Use Spawn Probability=false)
   &SpawnRate(
    SpawnGroup=1
    SpawnRate=&Multiply_Float(A=$System.TrailSpawnRate, B=$User.Trail2)
    Use Spawn Probability=false
   ) {module=SpawnRate001}
  ]
  particleSpawnScript [
   &InitializeParticle(
    Lifetime=0.25
    Ribbon Facing Mode=::NewEnumerator1
    Ribbon Facing Vector:Vector3f=$Engine.Owner.SystemZAxis
    Ribbon Width=32.5
    Ribbon Width Mode=::NewEnumerator1
    UsePositionOffset=false
    Write Material Random=true
   )
   &SkeletalMeshLocation(
    Bone / Socket Index=&SpawnGroup()
    Bone Sampling Mode=::NewEnumerator4
    MaskSpawnGroup=false
    Sampled Position Offset=<-5 0 0>
    Skeletal Mesh=$System.Skeletal Mesh
   )
   &AddVelocity(
    Module.Use Falloff Distance=false
    Override Seed=false
    Velocity Mode=::NewEnumerator1
    Velocity Origin=$Particles.SkeletalMeshLocation.SampledPosition
    Velocity Speed=125
   )
  ]
  particleUpdateScript [
   &ParticleState()
   &ScaleRibbonWidth(Ribbon Width Scale=&FloatFromCurve(FloatCurve=DataInterface() {propertyValues {Curve {keys [
         {arriveTangent=-0.3614, interpMode=RCIM_Cubic, leaveTangent=-0.3614, tangentMode=RCTM_User, time=0, value=0.1}
         {arriveTangent=4.75, interpMode=RCIM_Cubic, leaveTangent=4.75, time=0.2, value=0.7}
         {arriveTangent=7.6614, interpMode=RCIM_Cubic, leaveTangent=7.6614, tangentMode=RCTM_User, time=0.4, value=2}
         {arriveTangent=0.2573, interpMode=RCIM_Cubic, leaveTangent=0.2573, tangentMode=RCTM_User, time=0.7, value=2.2}
        ]}}}))
   &ScaleColor(
    Scale Alpha=&FloatFromCurve(FloatCurve=DataInterface() {propertyValues {Curve {keys [
         {arriveTangent=8.0122, interpMode=RCIM_Cubic, leaveTangent=8.0122, tangentMode=RCTM_User, time=0, value=0}
         <at 0.2 1 RCIM_Cubic>
         <at 1 0 RCIM_Cubic>
        ]}}})
    Scale RGB=&Multiply_Vector(
     A=&MakeVectorFromLinearColorRGB(Color=&MaskLinearColorBySpawnGroup(
       Masked Linear Color=&ColorFromCurve(ColorCurve=DataInterface() {propertyValues {
          AlphaCurve {keys [<at 0 1 RCIM_Cubic>]}
          BlueCurve {keys [<at 0.5245 0.581 RCIM_Cubic>,<at 1 0.2241 RCIM_Cubic>]}
          GreenCurve {keys [<at 0.5245 0.1714 RCIM_Cubic>,<at 1 0 RCIM_Cubic>]}
          RedCurve {keys [<at 0.5245 1 RCIM_Cubic>,<at 1 0.2604 RCIM_Cubic>]}
         }})
       Passthrough Linear Color=&Multiply_LinearColorByFloat(
        Linear Color=&ColorFromCurve(ColorCurve=DataInterface() {propertyValues {
           AlphaCurve {keys [<at 0 1 RCIM_Cubic>]}
           BlueCurve {keys [<at 0.5102 0.0527 RCIM_Cubic>,<at 1 0 RCIM_Cubic>]}
           GreenCurve {keys [<at 0.5102 0.6418 RCIM_Cubic>,<at 1 0.2739 RCIM_Cubic>]}
           RedCurve {keys [<at 0.5102 0.8438 RCIM_Cubic>,<at 1 0.375 RCIM_Cubic>]}
          }})
        Float=&FNTimeOfDayScalar(Night Scale=0.7, Morning Scale=1, Evening Scale=1)
       )
      ))
     B=&MakeVectorFromLinearColorRGB(Color=$System.ToD)
    )
    ScaleA=true
    ScaleRGB=true
   )
   &DynamicMaterialParameters(
    Index 0 Param 1=&RandomRangeFloat(Minimum=0, Maximum=0.5)
    Index 0 Param 2=&FloatFromCurve(FloatCurve=DataInterface() {propertyValues {Curve {keys [{arriveTangent=-3.1456, time=0, value=1}, <at 0.6 0.2>, {arriveTangent=0.0322, leaveTangent=0.0167, time=1, value=0}]}}})
    Index 0 Param 3=1
    Index 0 Param 4=1
    Module.Index0Float0Write=true
    Module.Index0Float1Write=true
    Module.Index0Float2Write=true
    Module.Index0Float3Write=true
   ) {module=DynamicMaterialParameters001}
   &self:ScratchModule(
    SkelMeshLoc:NiagaraPosition=$Particles.SkeletalMeshLocation.SampledPosition
    SpawnGroup:NiagaraInt32=&SpawnGroup()
    SpawnRate:NiagaraFloat=$System.TrailSpawnRate
   ) {enabled=false, module=ScratchModule}
   &self:ScratchModule_02(
    SkelMeshLoc:NiagaraPosition=$Particles.SkeletalMeshLocation.SampledPosition
    SpawnGroup:NiagaraInt32=&SpawnGroup()
    SpawnRate:NiagaraFloat=$System.TrailSpawnRate
   ) {enabled=false, module=ScratchModule_02}
   /SetVariables(Particles.Position=/Niagara/DynamicInputs/Helpers/MaskPosition(
     Mask Source:NiagaraInt32=0
     Mask Parameter:NiagaraInt32=?
     Mask:NiagaraInt32=0
     Masked Position:NiagaraPosition=$Particles.Ribbon1Pos
     Passthrough Position:NiagaraPosition=$Particles.Ribbon2Pos
    )) {enabled=false, isSetParameters=true, module=SetVariables_#4}
   &InheritSourceMovement(ApplyForce=false, ApplyPosition=true, ApplyPositionFalloff=false)
   &GravityForce(Coordinate Space=::Local, Gravity=<0 0 -500>)
   &Drag(Drag=2.2, Use Linear Drag=true, Use Rotational Drag=true)
   &SolveForcesAndVelocity(Clamp Velocity=false, Limit Acceleration=false)
  ]
  rendererClasses [&NiagaraRibbonRendererProperties]
  renderers [{rendererClass=&NiagaraRibbonRendererProperties}]
 }
 MainTrail001 {
  emitterUpdateScript [
   &EmitterState()
   &SpawnRate(SpawnRate=&Multiply_Float(A=$System.SmokeTrailSpawnRate, B=$User.SmokeTrail), Use Spawn Probability=false)
   &SpawnRate(
    SpawnGroup=1
    SpawnRate=&Multiply_Float(A=$System.SmokeTrailSpawnRate, B=$User.SmokeTrail2)
    Use Spawn Probability=false
   ) {module=SpawnRate001}
  ]
  particleSpawnScript [
   &InitializeParticle(
    Lifetime=0.275
    Ribbon Facing Mode=::NewEnumerator1
    Ribbon Facing Vector:Vector3f=$Engine.Owner.SystemZAxis
    Ribbon Width=37.5
    Ribbon Width Mode=::NewEnumerator1
    UsePositionOffset=false
    Write Material Random=true
   )
   &SkeletalMeshLocation(
    Bone / Socket Index=&SpawnGroup()
    Bone Sampling Mode=::NewEnumerator4
    MaskSpawnGroup=false
    Sampled Position Offset=<-5 0 0>
    Skeletal Mesh=$System.Skeletal Mesh
   )
   &AddVelocity(
    Module.Use Falloff Distance=false
    Override Seed=false
    Velocity Mode=::NewEnumerator1
    Velocity Origin=$Particles.SkeletalMeshLocation.SampledPosition
    Velocity Speed=125
   )
  ]
  particleUpdateScript [
   &ParticleState()
   &ScaleRibbonWidth(Ribbon Width Scale=&FloatFromCurve(FloatCurve=DataInterface() {propertyValues {Curve {
        defaultValue=0
        keys [
         {arriveTangent=0.0979, interpMode=RCIM_Cubic, leaveTangent=0.0979, tangentMode=RCTM_User, time=0, value=0}
         {arriveTangent=2.8207, interpMode=RCIM_Cubic, leaveTangent=2.8207, tangentMode=RCTM_User, time=0.45, value=1}
        ]
       }}}))
   &ScaleColor(
    Scale Alpha=&FloatFromCurve(
     FloatCurve=DataInterface() {propertyValues {Curve {keys [
         {arriveTangent=8.0122, interpMode=RCIM_Cubic, leaveTangent=8.0122, tangentMode=RCTM_User, time=0, value=0}
         <at 0.2 1 RCIM_Cubic>
         <at 1 0 RCIM_Cubic>
        ]}}}
     Scale Curve=0.6
    )
    Scale RGB=&Multiply_Vector(
     A=&MakeVectorFromLinearColorRGB(Color=&MaskLinearColorBySpawnGroup(
       Masked Linear Color=&ColorFromCurve(ColorCurve=DataInterface() {propertyValues {
          AlphaCurve {keys [<at 0 1 RCIM_Cubic>]}
          BlueCurve {keys [<at 0 0.8605 RCIM_Cubic>,<at 1 0.6705 RCIM_Cubic>]}
          GreenCurve {keys [<at 0 0 RCIM_Cubic>,<at 1 0.4051 RCIM_Cubic>]}
          RedCurve {keys [<at 0 1 RCIM_Cubic>,<at 1 0.7135 RCIM_Cubic>]}
         }})
       Passthrough Linear Color=&ColorFromCurve(ColorCurve=DataInterface() {propertyValues {
          AlphaCurve {keys [<at 0 1 RCIM_Cubic>]}
          BlueCurve {keys [<at 0 0 RCIM_Cubic>,<at 1 0.2044 RCIM_Cubic>]}
          GreenCurve {keys [<at 0 0.629 RCIM_Cubic>,<at 1 0.5763 RCIM_Cubic>]}
          RedCurve {keys [<at 0 1 RCIM_Cubic>,<at 1 0.7135 RCIM_Cubic>]}
         }})
      ))
     B=&MakeVectorFromLinearColorRGB(Color=$System.ToD)
    )
    ScaleA=true
    ScaleRGB=true
   )
   &DynamicMaterialParameters(
    Index 0 Param 1=&FloatFromCurve(FloatCurve=DataInterface() {propertyValues {Curve {
        defaultValue=1
        keys [
         {tangentMode=RCTM_User, time=0, value=1}
         {arriveTangent=-0.0405, interpMode=RCIM_Cubic, leaveTangent=-0.0405, tangentMode=RCTM_User, time=0.4, value=1}
         {arriveTangent=-1.3007, interpMode=RCIM_Cubic, leaveTangent=-1.3007, tangentMode=RCTM_User, time=1, value=0}
        ]
       }}})
    Index 0 Param 2=0.4
    Index 0 Param 3=-0.25
    Index 0 Param 4=1
    Module.Index0Float0Write=true
    Module.Index0Float1Write=true
    Module.Index0Float2Write=true
    Module.Index0Float3Write=true
   ) {module=DynamicMaterialParameters001}
   &self:ScratchModule(
    SkelMeshLoc:NiagaraPosition=$Particles.SkeletalMeshLocation.SampledPosition
    SpawnGroup:NiagaraInt32=&SpawnGroup()
    SpawnRate:NiagaraFloat=$System.SmokeTrailSpawnRate
   ) {enabled=false, module=ScratchModule}
   &self:ScratchModule_02(
    SkelMeshLoc:NiagaraPosition=$Particles.SkeletalMeshLocation.SampledPosition
    SpawnGroup:NiagaraInt32=&SpawnGroup()
    SpawnRate:NiagaraFloat=$System.SmokeTrailSpawnRate
   ) {enabled=false, module=ScratchModule_02}
   /SetVariables(Particles.Position=&MaskPosition(
     Mask Source:NiagaraInt32=0
     Mask Parameter:NiagaraInt32=?
     Mask:NiagaraInt32=0
     Masked Position:NiagaraPosition=$Particles.Ribbon1Pos
     Passthrough Position:NiagaraPosition=$Particles.Ribbon2Pos
    )) {enabled=false, isSetParameters=true, module=SetVariables_#4}
   &InheritSourceMovement(ApplyForce=false, ApplyPosition=true, ApplyPositionFalloff=false)
   &GravityForce(Coordinate Space=::Local, Gravity=<0 0 -500>)
   &Drag(Drag=2.2, Use Linear Drag=true, Use Rotational Drag=true)
   &SolveForcesAndVelocity(Clamp Velocity=false, Limit Acceleration=false)
  ]
  rendererClasses [&NiagaraRibbonRendererProperties]
  renderers [{rendererClass=&NiagaraRibbonRendererProperties}]
 }
 Smoke {
  emitterUpdateScript [
   &EmitterState()
   &SpawnRate(SpawnRate=&Multiply_Float(A=24, B=$User.Trail), Use Spawn Probability=false)
   &SpawnRate(SpawnGroup=1, SpawnRate=&Multiply_Float(A=24, B=$User.Trail2), Use Spawn Probability=false) {module=SpawnRate001}
  ]
  particleSpawnScript [
   &InitializeParticle(
    Color=&MaskLinearColorBySpawnGroup(
     Masked Linear Color=<rgba 1 0.0778 0.8737 1>
     Passthrough Linear Color=&Multiply_LinearColorByFloat(
      Linear Color=<rgba 1 0.7306 0 1>
      Float=&FNTimeOfDayScalar(Night Scale=0.85, Morning Scale=1, Evening Scale=1)
     )
    )
    Lifetime Max=0.6
    Lifetime Min=0.3
    Lifetime Mode=::NewEnumerator1
    Sprite Rotation Mode=::NewEnumerator2
    Sprite Size Mode=::NewEnumerator1
    Uniform Sprite Size Max=45
    Uniform Sprite Size Min=25
    UsePositionOffset=false
    Write Material Random=true
   )
   &SkeletalMeshLocation(
    Bone / Socket Index=&SpawnGroup()
    Bone Sampling Mode=::NewEnumerator4
    MaskSpawnGroup=false
    Skeletal Mesh=$System.Skeletal Mesh
   )
   &AddVelocity(
    Cone Axis=&ReturnMeshOrientationAxis(Mesh Orientation Axis=::XAxis, Mesh Orientation Quat=$Particles.MeshOrientation)
    Module.UseConeFalloff=false
    Module.UseVelDistribution=false
    Override Seed=false
    Velocity Mode=::NewEnumerator2
    Velocity Speed=&RandomRangeFloat(Minimum=-75, Maximum=-125)
   )
  ]
  particleUpdateScript [
   &ParticleState()
   &SubUVAnimation(
    "Sprite Renderer "=DataInterface() {propertyValues {SpriteRenderer=&self:Smoke_1.NiagaraSpriteRendererProperties_0}}
    UseEndFrame=false
    UseStartFrame=false
   )
   &ScaleSpriteSize(
    Uniform Curve Scale=1.45
    Uniform Curve Sprite Scale=DataInterface() {propertyValues {Curve {
       defaultValue=0
       keys [
        {arriveTangent=4.1785, interpMode=RCIM_Cubic, leaveTangent=4.1785, tangentMode=RCTM_User, time=0, value=0}
        {arriveTangent=1.1179, interpMode=RCIM_Cubic, leaveTangent=1.1179, tangentMode=RCTM_User, time=0.2042, value=0.6228}
        {arriveTangent=0.0224, interpMode=RCIM_Cubic, leaveTangent=0.0224, tangentMode=RCTM_User, time=1, value=1}
       ]
      }}}
   )
   &ScaleColor(
    Scale Alpha=&FloatFromCurve(
     FloatCurve=DataInterface() {propertyValues {Curve {
        defaultValue=0
        keys [
         {arriveTangent=9.0725, interpMode=RCIM_Cubic, leaveTangent=9.0725, tangentMode=RCTM_User, time=0, value=0}
         {arriveTangent=-2.5249, interpMode=RCIM_Cubic, leaveTangent=-2.5249, tangentMode=RCTM_User, time=0.1408, value=0.9847}
         {arriveTangent=-0.0405, interpMode=RCIM_Cubic, leaveTangent=-0.0405, tangentMode=RCTM_User, time=1, value=0}
        ]
       }}}
     Scale Curve=0.45
    )
    Scale RGB=&Multiply_Vector(A=<1 1 1>, B=&MakeVectorFromLinearColorRGB(Color=$System.ToD))
    ScaleA=true
    ScaleRGB=true
   )
   &Drag(Drag=&RandomRangeFloat(Minimum=0.8, Maximum=1.2), Use Linear Drag=true, Use Rotational Drag=true)
   &SolveForcesAndVelocity(Clamp Velocity=false, Limit Acceleration=false)
   &DynamicMaterialParameters(
    Index 0 Param 1=3
    Index 0 Param 2=1
    Index 0 Param 3=1
    Module.Index0Float0Write=true
    Module.Index0Float1Write=true
    Module.Index0Float2Write=true
    Module.Index0Float3Write=true
   )
  ]
  rendererClasses [&NiagaraSpriteRendererProperties]
  renderers [{rendererClass=&NiagaraSpriteRendererProperties}]
 }
 Spray {
  emitterUpdateScript [
   &EmitterState()
   &SpawnBurst_Instantaneous(Use Spawn Probability=false, UseLoopCountLimit=false)
   &SpawnBurst_Instantaneous(Spawn Group=1, Use Spawn Probability=false, UseLoopCountLimit=false) {module=SpawnBurst_Instantaneous001}
   &SpawnBurst_Instantaneous(Spawn Group=2, Use Spawn Probability=false, UseLoopCountLimit=false) {module=SpawnBurst_Instantaneous002}
   &SpawnBurst_Instantaneous(Spawn Group=3, Use Spawn Probability=false, UseLoopCountLimit=false) {module=SpawnBurst_Instantaneous004}
  ]
  particleSpawnScript [
   /SetVariables(Particles.VisibilityTag=&MaskIntBySpawnGroup(
     Masked Int=0
     Passthrough Int=&MaskIntBySpawnGroup(
      Masked Int=1
      Passthrough Int=&MaskIntBySpawnGroup(Masked Int=2, Passthrough Int=3, Spawn Group Mask=2)
      Spawn Group Mask=1
     )
     Spawn Group Mask=0
    )) {isSetParameters=true}
   &InitializeParticle(
    Color=&MaskLinearColorBySpawnGroup(Masked Linear Color=<rgba 5 0.3892 4.3685 1>, Passthrough Linear Color=<rgba 5 4.0742 0 1>)
    Lifetime=7.933
    Mesh Scale=<0.15 0.35 0.15>
    Mesh Scale Mode=::NewEnumerator4
    UsePositionOffset=false
    Write Material Random=true
   )
   &InitialMeshOrientation(Mesh Axis To Orient=<-1 0 0>, Mesh Orientation Mode=::NewEnumerator2, Use Rotation Vector=false)
  ]
  particleUpdateScript [
   &ParticleState()
   &ScaleMeshSize(Scale Factor=&MakeVector(X=1, Y=&Waveform("Waveform [1]"=::NewEnumerator2, "[1] Frequency"=1, "[1] Amplitude Min/Max"=<0.8 1.2>), Z=1))
   &SkeletalMeshLocation(
    Bone / Socket Index=&SpawnGroup()
    Bone Sampling Mode=::NewEnumerator4
    MaskSpawnGroup=false
    Skeletal Mesh=$System.Skeletal Mesh
   )
   &UpdateMeshOrientation(
    Facing Direction=&ReturnMeshOrientationAxis(Mesh Orientation Axis=::XAxis, Mesh Orientation Quat=$Particles.MeshOrientation)
    Facing Interpolation=false
    Orientation Method=::NewEnumerator1
    Override Quat=false
    TransformFacing=false
    TransformUp=false
    Up Direction:Vector3f=<0 0 -1>
   ) {enabled=false}
   &DynamicMaterialParameters(
    Index 0 Param 1=&FloatFromCurve(FloatCurve=DataInterface() {propertyValues {Curve {keys [<at 0 0>,<at 1 99>]}}})
    Module.Index0Float0Write=true
    Module.Index0Float1Write=true
    Module.Index0Float2Write=true
    Module.Index0Float3Write=true
   )
   /SetVariables(
    Particles.BurstSprayColor:LinearColor=&MultiplyLinearColor(
     Linear Color=&MaskLinearColorBySpawnGroup(Masked Linear Color=<rgba 5 0.3892 4.3685 1>, Passthrough Linear Color=<rgba 5 4.0742 0 1>, Spawn Group Mask=2)
     Scale Factor=/Niagara/DynamicInputs/Multiply/Multiply_Vector4_ByFloat(Vector4:Vector4f=<1 1 1 1>, Float:NiagaraFloat=$User.SmokeCone)
    )
    Particles.BurstSprayDP:Vector4f=<2 2 1 0>
    Particles.BurstSpraySize:Vector3f=<0.075 0.2 0.075>
   ) {isSetParameters=true}
   &ScaleColor(
    Scale Alpha=&Multiply_Float(A=&Multiply_Float(A=1, B=1), B=$User.Spray)
    Scale RGB=&Multiply_Vector(A=<1 1 1>, B=&MakeVectorFromLinearColorRGB(Color=$System.ToD))
    ScaleA=true
    ScaleRGB=true
   )
  ]
  rendererClasses [&NiagaraMeshRendererProperties]
  renderers [
   {rendererClass=&NiagaraMeshRendererProperties, rendererIndex=0}
   {rendererClass=&NiagaraMeshRendererProperties, rendererIndex=1}
   {rendererClass=&NiagaraMeshRendererProperties, rendererIndex=2}
   {rendererClass=&NiagaraMeshRendererProperties, rendererIndex=3}
  ]
 }
}
summary {
 systemName=NS_GraffitiTon_Trail
 emitters [
  {emitterName=MainTrail, rendererClasses [&NiagaraRibbonRendererProperties]}
  {emitterName=Spray, rendererClasses [&NiagaraMeshRendererProperties]}
  {emitterName=BurstSmoke, rendererClasses [&NiagaraSpriteRendererProperties]}
  {emitterName=Drops, simTarget=GPUComputeSim, rendererClasses [&NiagaraSpriteRendererProperties]}
  {emitterName=Decal, rendererClasses [&NiagaraDecalRendererProperties]}
  {emitterName=Smoke, rendererClasses [&NiagaraSpriteRendererProperties]}
  {emitterName=MainTrail001, rendererClasses [&NiagaraRibbonRendererProperties]}
 ]
 userVariables [
  User.bIsFrontEnd {type=NiagaraBool}
  User.bIsFrontEndPreview {type=NiagaraBool}
  User.Burst {type=NiagaraFloat, defaultValue=0, description=""}
  User.Burst2 {type=NiagaraFloat, defaultValue=0, description=""}
  User.RibbonIDOffset {type=NiagaraInt32}
  User.SmokeCone {type=NiagaraFloat, defaultValue=0, description=""}
  User.SmokeTrail {type=NiagaraFloat, defaultValue=0, description=""}
  User.SmokeTrail2 {type=NiagaraFloat, defaultValue=0, description=""}
  User.Spray {type=NiagaraFloat, defaultValue=0, description=""}
  User.Trail {type=NiagaraFloat, defaultValue=0, description=""}
  User.Trail2 {type=NiagaraFloat, defaultValue=0, description=""}
 ]
}
